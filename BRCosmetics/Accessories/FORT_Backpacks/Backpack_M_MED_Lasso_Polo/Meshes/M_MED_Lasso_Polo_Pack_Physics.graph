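# Physics asset for the Lasso Polo backpack cosmetic, giving its pack, flap, ball and D-ring bones simple collision bodies and joints.
# The flap and ball are each constrained to the pack capsule, and the D-ring hangs off the ball in a short chain.
# Joint limits are mostly tight, with one wider swing axis on the flap and a free swing axis on the D-ring.
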PhysicsAsset /BRCosmetics/Accessories/FORT_Backpacks/Backpack_M_MED_Lasso_Polo/Meshes/M_MED_Lasso_Polo_Pack_Physics
bodies {
 dyn_DRing [{shapeType=Capsule, length=2, radius=1, center=<1.7475 0 0>}]
 dyn_ball [{shapeType=Capsule, length=0, radius=8}]
 dyn_flap [{
   shapeType=Box
   extentX=1.6589
   extentY=8.112
   extentZ=20.3954
   shapeName=dyn_flap_box
   center=<4.2874 -0.8956 0.0003>
   rotation=<rot 0.0695 -81.1468 0.059>
  }]
 pack [{
   shapeType=Capsule
   length=19.9057
   radius=7.7056
   shapeName=pack_capsule
   center=<0.0028 -6.6519 -8.9446>
   rotation=<rot 8.0828 -89.9903 0.0557>
  }]
}
constraints [
 {bone1Name=dyn_DRing, bone2Name=dyn_ball, swing1LimitDegrees=45, swing1Motion=Free, swing2LimitDegrees=1, twistLimitDegrees=1}
 {bone1Name=dyn_flap, bone2Name=pack, swing1LimitDegrees=44.4, swing2LimitDegrees=3, twistLimitDegrees=3}
 {bone1Name=dyn_ball, bone2Name=pack, swing1LimitDegrees=5, swing2LimitDegrees=5, twistLimitDegrees=5}
]
